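; Cosmetic Blueprint for the AvenuePole backpack that adds visual effects in reaction to the wearer's eliminations.
; It watches the AthenaKills stat and, on the first value above zero, spawns a Niagara burst attached to the idle VFX component.
; A separate loop event, with no caller shown, reactivates the idle VFX and stamps game time into the mesh's Timestamp_Elim material parameter.
; On destruction in the front end, it clears a stored timer.
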(event EventOnPostResIn
  (|CallParentFunction)
  (bind (_statname _statvalue) (Cosmetics|WatchCosmeticStat (CustomCharacter|GetAssociatedPlayerPawn) "(TagName=\"StatManager.AthenaKills\")" -1)
    (:OnStatChanged
      (if (> _statvalue 0)
        (Utilities|FlowControl|DoOnce false
          (:Completed
            (Niagara|SpawnSystemAttached "/BRCosmetics/Accessories/FORT_Backpacks/Backpack_M_MED_AvenuePole/FX/NS_Backpack_AvenuePole_Burst.NS_Backpack_AvenuePole_Burst" (CallFunction|GetIdleVFXComponent))
            (Utilities|FlowControl|Delay 0.5)))))))

(event Custom|CE%20FN%20Loop
  (Components|Activation|Activate (CallFunction|GetIdleVFXComponent) true)
  (Rendering|Material|SetScalarParameterValueonMaterials (CustomCharacter|GetSkeletalMeshforPartonOwner) "Timestamp_Elim" (Utilities|Time|GetGameTimeinSeconds)))

(event EventDestroyed
  (|CallParentFunction)
  (if (CustomCosmeticModifier|IsinFrontEnd)
    (Utilities|Time|ClearandInvalidateTimerbyHandle (Variables|Default|GetTimer))))

(fn ConstructionScript
  (|CallParentFunction))
